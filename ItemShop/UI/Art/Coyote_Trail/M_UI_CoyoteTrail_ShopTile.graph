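# UI material for the Coyote Trail item shop tile, drawing an animated green pattern on a dark background.
# Two pattern textures share scaled, rotated UVs and tint the color layers, which blend with a radial gradient.
# A time-scrolled, rotated sine band adds a moving highlight, gated by a slower sine pulse driven by the tenminutes parameter.
# Parameterized horizontal and vertical falloff masks darken the edges; several leftover nodes feed nothing.
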Material /ItemShop/UI/Art/Coyote_Trail/M_UI_CoyoteTrail_ShopTile
graph {
 properties {bUsedWithStaticMesh=false, materialDomain=MD_UI, pixelDepthOffsetMode=PDOM_Legacy, usedShadingModels=Default Lit}
 nodes [
  n1 = Time()
  n2 = ScalarParameter() {defaultValue=1, parameterName=tenminutes}
  n3 = Multiply(A=Sine(None=Multiply(A=@n1, B=@n2)) {period=5}) {constB=3}
  Ceil(None=@n3)
  Constant3Vector() {constant=<rgba 0 0.0769 0.0833 0>}
  Constant3Vector() {constant=<rgba 0 0.0769 0.0833 0>}
  Constant3Vector() {constant=<rgba 0 0.0094 0.0104 0>}
  Constant3Vector() {constant=<rgba 0.0276 0.1615 0.0338 1>}
  Constant3Vector() {constant=<rgba 0 0.5833 0.3987 1>}
  n4 = Saturate(None=SmoothStep(Value=@n3) {constMax=8, constMin=0.5})
  n5 = MaterialFunctionCall() {materialFunction=&MF_UI_UVScaling}
  n6 = Time()
  n7 = Multiply(A=Frac(None=OneMinus(None=@n6)))
  n8 = Constant()
  n9 = MaterialFunctionCall(
   UVs=Add(A=ConstantBiasScale(None=@n5.Scaled To Fill) {scale=0.24}, B=AppendVector(A=@n7, B=@n8))
   "Rotation Angle (0-1)"=Constant() {r=0.358}
  ) {materialFunction=&CustomRotator}
  n10 = Multiply(A=ComponentMask(None=Subtract(A=@n9.Rotated Values) {constB=0}) {r=true}) {constB=0.5}
  n11 = Constant() {r=0.0025}
  n12 = Multiply(A=Subtract(A=Sine(None=@n10), B=OneMinus(None=@n11)) {constB=0.25}, B=Divide(B=@n11) {constB=0.5})
  n13 = Multiply(A=Saturate(None=Multiply(A=@n12) {constB=2})) {constB=10}
  n14 = MaterialFunctionCall() {materialFunction=&MF_UI_UVScaling}
  n15 = Add(
   A=@n14.ScaleY
   B=MaterialFunctionCall(X=Constant() {r=0.95}, Y=Constant()) {materialFunction=&MakeFloat2}.Result
  )
  n16 = ScalarParameter() {defaultValue=1.0131, parameterName=PatternSize}
  n17 = MaterialFunctionCall(
   UVs=MaterialFunctionCall(UVs=@n15, Texture Scale=@n16) {materialFunction=&ScaleUVsByCenter}.UVs
   "Rotation Angle (0-1)"=Constant() {r=0.25}
  ) {materialFunction=&CustomRotator}
  n18 = TextureSample(UVs=@n17.Rotated Values) {texture=/OfferCatalog/Art/Coyote_Trail/T_UI_CoyoteTrailPattern}
  n19 = LinearInterpolate(B=LinearInterpolate(B=Constant3Vector() {constant=<rgba 0.0137 0.3515 0.2384 1>}, Alpha=Multiply(A=@n4, B=@n13)), Alpha=@n18.R)
  n20 = LinearInterpolate(
   A=LinearInterpolate(
    A=@n19
    B=Constant3Vector() {constant=<rgba 0.01 0.0176 0.0149 0>}
    Alpha=TextureSample(UVs=@n17.Rotated Values) {texture=/OfferCatalog/Art/Coyote_Trail/T_UI_CoyoteTrailPattern_02}.R
   )
   B=LinearInterpolate(A=Multiply(A=@n19) {constB=7.6538}, B=Constant3Vector() {constant=<rgba 0.0186 0.1059 0.0653 0>}, Alpha=@n18.R)
  )
  n21 = Constant3Vector() {constant=<rgba 0.0087 0.0087 0.0087 0>}
  n22 = Constant3Vector() {constant=<rgba 0.0312 0.0356 0.0434 0>}
  n23 = Constant() {r=1}
  n24 = Constant() {r=0.5}
  n25 = MaterialFunctionCall(
   UV=Multiply(
    A=MaterialFunctionCall(X=@n23, Y=@n24) {materialFunction=&MakeFloat2}.Result
    B=@n14.ScaleY
   )
   GradientStart=Constant()
   GradientEnd=Constant() {r=1.0948}
  ) {materialFunction=&MF_UI_RadialGradient}
  n26 = MaterialFunctionCall() {materialFunction=&MF_UI_UVScaling}
  n27 = Abs(None=ConstantBiasScale(None=@n26.Scaled To Fill) {bias=-0.5, scale=2})
  n28 = Multiply(
   A=Subtract(A=ComponentMask(None=@n27) {g=true}, B=ScalarParameter() {defaultValue=0.0511, parameterName=Subtract})
   B=ScalarParameter() {defaultValue=1.0721, parameterName=Falloff}
  )
  n29 = Abs(None=ConstantBiasScale(None=@n26.Scaled To Fill) {bias=-0.5, scale=2})
  n30 = Multiply(
   A=Subtract(A=ComponentMask(None=@n29) {r=true}, B=ScalarParameter() {defaultValue=0.302, parameterName=SubtractH})
   B=ScalarParameter() {defaultValue=0.8595, parameterName=FalloffH}
  )
  n31 = LinearInterpolate(
   A=Multiply(A=Saturate(None=@n28), B=ScalarParameter() {defaultValue=1, parameterName=Intensity})
   B=Multiply(A=Saturate(None=@n30), B=ScalarParameter() {defaultValue=2.3306, parameterName=IntensityH})
  )
  LinearInterpolate(A=LinearInterpolate(A=Multiply(A=@n20) {constB=4}, B=LinearInterpolate(A=@n21, B=@n22, Alpha=@n25.Smooth)), B=Constant3Vector(), Alpha=@n31)
  MaterialFunctionCall(UVs=@n15, Texture Scale=@n16) {materialFunction=&ScaleUVsByCenter}
  Multiply()
  OneMinus(None=@n31)
  n32 = Sine(None=Multiply(A=Time()))
  Saturate(None=Multiply(A=Add(A=@n32)) {constB=2})
  n33 = Sine(None=Multiply(A=Time()) {constB=0.01})
  Saturate(None=Multiply(A=Add(A=@n33)) {constB=0.5})
 ]
}
